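# Animation Blueprint for the CiderPetalFlax back bling that makes its five pieces oscillate after enemy eliminations.
# Time-based sine offsets drive additive bone-space translations on the piece bones, eased through interpolated alpha.
# Elimination events advance a clamped stage counter, but every switch case triggers the same timed pulse; the front end loops a similar pulse.
# The pack bone has a constant offset, plus a shift and roll while skydiving.
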AnimBlueprint /BRCosmetics/Accessories/FORT_Backpacks/Backpack_M_MED_CiderPetalFlax/Meshes/M_MED_CiderPetalFlax_Pack_AnimBP
target_skeleton=&/M_MED_CiderPetalFlax_Pack_Skeleton
graphs [
 AnimGraph {nodes [
   n1 = Get isEnemyPlayerEliminated() -> (isEnemyPlayerEliminated:double)
   n2 = "Transform (Modify) Bone - Bone: dyn_pieace_01"(
    ComponentPose=Local To Component(LocalPose=Input Pose())
    Alpha=@n1
    Translation=Get Pieces move() -> (Pieces move:Vector)
    Rotation="0,0,0"
    Scale="1,1,1"
   ) {
    translationMode=BMM_Additive
    translationSpace=BCS_BoneSpace
    alphaScaleBiasClamp {bInterpResult=true, interpSpeedDecreasing=5, interpSpeedIncreasing=4}
    boneToModify {boneName=dyn_pieace_01}
   }
   n3 = "Transform (Modify) Bone - Bone: dyn_pieace_02"(ComponentPose=@n2, Alpha=@n1, Translation=Get Pieces move1() -> (Pieces move1:Vector), Rotation="0,0,0", Scale="1,1,1") {
    translationMode=BMM_Additive
    translationSpace=BCS_BoneSpace
    alphaScaleBiasClamp {bInterpResult=true, interpSpeedDecreasing=5, interpSpeedIncreasing=4}
    boneToModify {boneName=dyn_pieace_02}
   }
   n4 = "Transform (Modify) Bone - Bone: dyn_pieace_03"(ComponentPose=@n3, Alpha=@n1, Translation=Get Pieces move2() -> (Pieces move2:Vector), Rotation="0,0,0", Scale="1,1,1") {
    translationMode=BMM_Additive
    translationSpace=BCS_BoneSpace
    alphaScaleBiasClamp {bInterpResult=true, interpSpeedDecreasing=5, interpSpeedIncreasing=4}
    boneToModify {boneName=dyn_pieace_03}
   }
   n5 = "Transform (Modify) Bone - Bone: dyn_pieace_04"(ComponentPose=@n4, Alpha=@n1, Translation=Get Pieces move3() -> (Pieces move3:Vector), Rotation="0,0,0", Scale="1,1,1") {
    translationMode=BMM_Additive
    translationSpace=BCS_BoneSpace
    alphaScaleBiasClamp {bInterpResult=true, interpSpeedDecreasing=5, interpSpeedIncreasing=4}
    boneToModify {boneName=dyn_pieace_04}
   }
   n6 = "Transform (Modify) Bone - Bone: dyn_pieace_05"(ComponentPose=@n5, Alpha=@n1, Translation=Get Pieces move4() -> (Pieces move4:Vector), Rotation="0,0,0", Scale="1,1,1") {
    translationMode=BMM_Additive
    translationSpace=BCS_BoneSpace
    alphaScaleBiasClamp {bInterpResult=true, interpSpeedDecreasing=5, interpSpeedIncreasing=4}
    boneToModify {boneName=dyn_pieace_05}
   }
   n7 = "Transform (Modify) Bone - Bone: pack"(ComponentPose=@n6, Translation="0,1.5,0", Rotation="0,0,0") {
    rotationMode=BMM_Additive
    rotationSpace=BCS_BoneSpace
    translationMode=BMM_Additive
    translationSpace=BCS_BoneSpace
    boneToModify {boneName=pack}
    translation {y=1.5}
   }
   n8 = "Transform (Modify) Bone - Bone: pack"(ComponentPose=@n7, Alpha=Get Skydiving() -> (Skydiving:double), Translation="0,-2,0", Rotation="0,0,10") {
    rotationMode=BMM_Additive
    rotationSpace=BCS_BoneSpace
    translationMode=BMM_Additive
    translationSpace=BCS_BoneSpace
    boneToModify {boneName=pack}
    rotation {roll=10}
    translation {y=-2}
   }
   Output Pose(Result=Component To Local(ComponentPose=@n8)) {layerGroup=None}
  ]}
 EventGraph {nodes [
   Event BlueprintUpdateAnimation()
   > n1 = Sequence!(execute) -> (then_0, then_1, then_2:Exec)
   n2 = GetGameTimeInSeconds()
   n3 = float + float(A="Sin (Radians)"(A=float * float(A=@n2, B=4) {is=title K2Node_PromotableOperator}), B=1.7)
   Set Pieces move(execute=@n1.then_1, Pieces move:Vector=MakeVector(X=float * float(A=@n3, B=2), Y=0)) -> (Output_Get:Vector)
   n4 = float + float(A="Sin (Radians)"(A=float * float(A=@n2, B=4)), B=3.5)
   > Set Pieces move1(Pieces move1:Vector=MakeVector(X=float * float(A=@n4, B=1), Y=0)) -> (Output_Get:Vector)
   n5 = float * float(A=@n2, B=4)
   n6 = float + float(A="Sin (Radians)"(A=@n5), B=4)
   n7 = float + float(A="Sin (Radians)"(A=@n5), B=-1.2)
   > Set Pieces move2(Pieces move2:Vector=MakeVector(X=float * float(A=@n6, B=1), Y=float * float(A=@n7, B=0.3))) -> (Output_Get:Vector)
   n8 = float * float(A=@n2, B=4)
   n9 = float + float(A="Sin (Radians)"(A=@n8), B=2.5)
   n10 = float + float(A="Sin (Radians)"(A=@n8), B=1.3)
   > Set Pieces move3(Pieces move3:Vector=MakeVector(X=float * float(A=@n9, B=1.5), Y=float * float(A=@n10, B=1.5))) -> (Output_Get:Vector)
   n11 = float * float(A=@n2, B=4)
   n12 = float + float(A="Sin (Radians)"(A=@n11), B=2)
   n13 = float + float(A="Sin (Radians)"(A=@n11), B=1)
   > Set Pieces move4(Pieces move4:Vector=MakeVector(X=float * float(A=@n12, B=2), Y=float * float(A=@n13, B=-0.2))) -> (Output_Get:Vector)
   Set Skydiving(execute=@n1.then_0, Skydiving:double="To Float (Boolean)"(InBool=Get bIsSkyDiving())) -> (Output_Get:double)
   n14 = Event BlueprintProcessAnimInputEvent()
   > Branch(Condition=IsNamed(self=@n14.AnimInputEvent, InEventName=aie.EnemyPlayerEliminated))
   > n15 = Increment Int(Value=Get EliminatedStage() -> (EliminatedStage:int=0))
   > Set EliminatedStage(EliminatedStage:int="Clamp (Integer)"(Value=@n15, Min=Get EliminatedStage() -> (EliminatedStage=0), Max=Get MaxEliminatedStage() -> (MaxEliminatedStage:int=0)) -> (ReturnValue=0)) -> (Output_Get:int=0)
   > n16 = Switch on Int!(execute:Exec, Selection:int=Get EliminatedStage() -> (EliminatedStage=0)) -> (Default:Exec, 0:Exec, 1:Exec, 2:Exec, 3:Exec, 4:Exec)
   Set isEnemyPlayerEliminated(execute=[@n16.0, @n16.Default], isEnemyPlayerEliminated:double=1) -> (Output_Get:double)
   > Delay(Duration=2.85)
   > Set isEnemyPlayerEliminated(isEnemyPlayerEliminated=0)
   Set isEnemyPlayerEliminated(execute=@n16.1, isEnemyPlayerEliminated=1)
   > Delay(Duration=2.85)
   > Set isEnemyPlayerEliminated(isEnemyPlayerEliminated=0)
   Set isEnemyPlayerEliminated(execute=@n16.2, isEnemyPlayerEliminated=1)
   > Delay(Duration=2.85)
   > Set isEnemyPlayerEliminated(isEnemyPlayerEliminated=0)
   Set isEnemyPlayerEliminated(execute=@n16.3, isEnemyPlayerEliminated=1)
   > Delay(Duration=2.85)
   > Set isEnemyPlayerEliminated(isEnemyPlayerEliminated=0)
   Set isEnemyPlayerEliminated(execute=@n16.4, isEnemyPlayerEliminated=1)
   > Delay(Duration=2.85)
   > Set isEnemyPlayerEliminated(isEnemyPlayerEliminated=0)
   Event OnCharacterCustomizationLoaded()
   > Branch(Condition=Get bIsInFrontEnd())
   > n17 = Delay(Duration=1)
   Set isEnemyPlayerEliminated(execute=[@n17.then, @n18.then], isEnemyPlayerEliminated=1)
   > Delay(Duration=2.45)
   > Set isEnemyPlayerEliminated(isEnemyPlayerEliminated=0)
   > n18 = Delay(Duration=3.5)
   TryGetPawnOwner()
  ]}
]
variables [
 EliminatedStage {type=integer}
 isEnemyPlayerEliminated {type=number}
 MaxEliminatedStage {type=integer}
 Pieces move {type=Vector}
 Pieces move1 {type=Vector}
 Pieces move2 {type=Vector}
 Pieces move3 {type=Vector}
 Pieces move4 {type=Vector}
 Pieces move5 {type=Vector}
 ResetTimerActive {type=boolean}
 Skydiving {type=number}
 TimeSinceMaxStage {type=number}
]
